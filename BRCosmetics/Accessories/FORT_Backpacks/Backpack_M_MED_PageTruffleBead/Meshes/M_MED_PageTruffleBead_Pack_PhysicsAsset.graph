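# Physics asset for the M_MED_PageTruffleBead backpack cosmetic mesh, giving the pack a simple collision shape.
# A single body on the pack bone holds one rotated capsule, offset slightly from the bone.
# There are no constraints, so the asset defines no jointed or dangling parts.
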PhysicsAsset /BRCosmetics/Accessories/FORT_Backpacks/Backpack_M_MED_PageTruffleBead/Meshes/M_MED_PageTruffleBead_Pack_PhysicsAsset
bodies {pack [{shapeType=Capsule, length=14.97, radius=13.968, center=<-0.1824 -6.6409 -2.1208>, rotation=<rot 0 91.7157 -88.8637>}]}
constraints []
